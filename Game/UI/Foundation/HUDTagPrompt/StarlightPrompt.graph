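# StarlightPrompt is a HUD tag prompt widget showing an uppercase title and description above a Starlight banner image, with a glow shadow.
# OpenWidget fills both texts from prompt data, loads the soft-referenced banner texture and plays Intro, which fades, scales and slides elements in.
# A retriggerable delay then plays Outro, an empty timed animation whose finish event is rebound on each open to an unnamed delegate.
# CloseWidget collapses the widget.
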WidgetBlueprint /Game/UI/Foundation/HUDTagPrompt/StarlightPrompt
animations [
 Intro {
  binding_count=3
  length_seconds=1.25
  object=&self:Intro
  sequence {bindings [
    SizeBox_5 {
     id=#1
     possessed_class=SizeBox
     tracks [{
       type=MovieScene2DTransformTrack
       sections [{
         type=MovieScene2DTransformSection
         end_frame=25
         start_frame=0
         channels [
          Scale.X {type=MovieSceneScriptingFloatChannel, keys [[0,0],[15,0],[25,1]]}
          Scale.Y {type=MovieSceneScriptingFloatChannel, keys [[0,0],[15,0],[25,1]]}
         ]
        }]
      }]
    }
    Overlay_0 {
     id=#2
     possessed_class=Overlay
     tracks [{
       type=MovieSceneFloatTrack
       display_name=Render Opacity
       sections [{type=MovieSceneFloatSection, end_frame=10, start_frame=0, channels [None {type=MovieSceneScriptingFloatChannel, keys [[0,0],[10,1]]}]}]
      }]
    }
    ImagePromptBackground {
     id=#3
     possessed_class=Image
     tracks [
      {
       type=MovieScene2DTransformTrack
       sections [{
         type=MovieScene2DTransformSection
         end_frame=20
         start_frame=0
         channels [
          Translation.X {type=MovieSceneScriptingFloatChannel, const [0,2]}
          Translation.Y {type=MovieSceneScriptingFloatChannel, keys [[0,200],[15,0]]}
          Angle {type=MovieSceneScriptingFloatChannel}
          Scale.X {type=MovieSceneScriptingFloatChannel}
          Scale.Y {type=MovieSceneScriptingFloatChannel}
          Shear.X {type=MovieSceneScriptingFloatChannel}
          Shear.Y {type=MovieSceneScriptingFloatChannel}
         ]
        }]
      }
      {
       type=MovieSceneColorTrack
       display_name=Color and Opacity
       sections [{
         type=MovieSceneColorSection
         end_frame=20
         start_frame=0
         channels [
          Color.R {type=MovieSceneScriptingFloatChannel}
          Color.G {type=MovieSceneScriptingFloatChannel}
          Color.B {type=MovieSceneScriptingFloatChannel}
          Color.A {type=MovieSceneScriptingFloatChannel, keys [[0,0],[15,1]]}
         ]
        }]
      }
     ]
    }
   ]}
 }
 Outro {binding_count=0, length_seconds=5, object=&self:Outro, sequence {bindings []}}
]
graphs [EventGraph {nodes [
   n1 = Event OpenWidget!() -> (OutputDelegate, then, PrompData:"Fort HUDTag Prompt Data Structure (by ref)") {is=title K2Node_Event}
   > Parent: OpenWidget!(execute, PrompData:"Fort HUDTag Prompt Data Structure (by ref)"=@n1.PrompData) -> (then) {is=title K2Node_CallParentFunction}
   n2 = Break Fort HUDTag Prompt Data!(FortHUDTagPromptData:"Fort HUDTag Prompt Data Structure (by ref)"=@n1.PrompData) -> (
    VolumeTag:Gameplay Tag Structure
    Text_PromptTitle:Text
    Text_PromptDescription:Text
    bIsEnterData:bool=~
    WidgetClass:User Widget Class Reference
   ) {is=title K2Node_BreakStruct}
   > "SetText (Text)"(self=Get Text_Title() -> (Text_Title:Common Text Object Reference), InText=@n2.Text_PromptTitle)
   > "SetText (Text)"(self=Get Text_Description() -> (Text_Description:Common Text Object Reference), InText=@n2.Text_PromptDescription)
   > UnbindAllFromAnimationFinished!(
    execute
    self:User Widget Object Reference
    Animation:Widget Animation Object Reference=Get Outro() -> (Outro:Widget Animation Object Reference)
   ) -> (then)
   > StopAnimation(InAnimation=Get Outro())
   > BindToAnimationFinished(Animation=Get Outro(), Delegate=Create Event())
   > SetBrushFromSoftTexture!(
    execute
    self:Image Object Reference=Get ImagePromptBackground() -> (ImagePromptBackground:Image Object Reference)
    SoftTexture:Texture 2D Soft Object Reference=Get ImageSoftRef() -> (ImageSoftRef:Texture 2D Soft Object Reference)
    bMatchSize:bool=false
   ) -> (then)
   > SetVisibility(InVisibility=HitTestInvisible)
   > PlayAnimationForward(InAnimation=Get Intro())
   > RetriggerableDelay(Duration=7)
   > PlayAnimation(InAnimation=Get Outro())
   Event CloseWidget() {is=title K2Node_Event}
   > Parent: CloseWidget() {is=title K2Node_CallParentFunction}
   > SetVisibility(InVisibility=Collapsed)
  ]}]
variables [ImageSoftRef {type=Texture2D, default=/Game/UI/Foundation/HUDTagPrompt/Materials/T_UI_Starlight_Prompt}]
widgets {
 info {inheritedWidgetCount=1, parentClass=/Script/FortniteUI.FortHUDTagPromptWidgetBase, rootWidgetClass=&Overlay, widgetCount=12}
 widgets [
  Overlay_0 {widgetClassPath=&Overlay, widget=Overlay_0}
  VerticalBox_1 {widgetClassPath=&VerticalBox, parent=Overlay_0, widget=VerticalBox_1, slot {class=OverlaySlot}}
  SizeBox_5 {
   widgetClassPath=&SizeBox
   parent=VerticalBox_1
   widget=SizeBox_5
   properties {bOverride_MaxDesiredWidth=true, maxDesiredWidth=650}
   slot {class=VerticalBoxSlot}
  }
  ScaleBox_6 {widgetClassPath=&ScaleBox, parent=SizeBox_5, widget=ScaleBox_6, slot {class=SizeBoxSlot}}
  VerticalBox_0 {widgetClassPath=&VerticalBox, parent=ScaleBox_6, widget=VerticalBox_0, slot {class=ScaleBoxSlot}}
  Text_Title {
   widgetClassPath=&CommonTextBlock
   bIsVariable=true
   parent=VerticalBox_0
   widget=Text_Title
   properties {
    justification=Center
    text=TOP TEXT
    textTransformPolicy=ToUpper
    font {
     bForceMonospaced=false
     bMaterialIsStencil=false
     fontMaterial=null
     fontObject=&Burbank
     letterSpacing=0
     monospacedWidth=1
     size=35
     skewAmount=0
     typefaceFontName=BurbankRegular-Black
     outlineSettings {
      bApplyOutlineToDropShadows=false
      bMiteredCorners=false
      bSeparateFillAlpha=false
      outlineMaterial=null
      outlineSize=1
      outlineColor=<rgba 0.6654 0 1 1>
     }
    }
    shadowColorAndOpacity=<rgba 0.3677 0 0.5521 0>
    shadowOffset=<0 3.106>
   }
   slot {class=VerticalBoxSlot}
  }
  Text_Description {
   widgetClassPath=&CommonTextBlock
   bIsVariable=true
   parent=VerticalBox_0
   widget=Text_Description
   properties {
    justification=Center
    text=Bottom text
    textTransformPolicy=ToUpper
    font {
     bForceMonospaced=false
     bMaterialIsStencil=false
     fontMaterial=null
     fontObject=&Burbank
     letterSpacing=0
     monospacedWidth=1
     size=16
     skewAmount=0
     typefaceFontName=Burbank-Black
     outlineSettings {
      bApplyOutlineToDropShadows=false
      bMiteredCorners=false
      bSeparateFillAlpha=false
      outlineMaterial=null
      outlineSize=0
      outlineColor=<rgba 0.6654 0 1 1>
     }
    }
    shadowOffset=<0 0>
   }
   slot {class=VerticalBoxSlot, properties {padding=<ltrb 0 -10 0 0>}}
  }
  SizeBox_6 {
   widgetClassPath=&SizeBox
   parent=VerticalBox_1
   widget=SizeBox_6
   properties {bOverride_MaxDesiredWidth=true, maxDesiredWidth=500}
   slot {class=VerticalBoxSlot, properties {verticalAlignment=VAlign_Bottom}}
  }
  ScaleBox_97 {widgetClassPath=&ScaleBox, parent=SizeBox_6, widget=ScaleBox_97, properties {stretch=ScaleToFitX}, slot {class=SizeBoxSlot}}
  ImagePromptBackground {
   widgetClassPath=&Image
   bIsVariable=true
   parent=ScaleBox_97
   widget=ImagePromptBackground
   properties {brush {
     bIsDynamicallyLoaded=false
     drawAs=Image
     imageType=NoImage
     mirroring=NoMirror
     resourceName=None
     resourceObject=null
     tiling=NoTile
     imageSize=<693 106>
     margin=<ltrb 0 0 0 0>
     outlineSettings {
      bUseBrushTransparency=false
      roundingType=HalfHeightRadius
      width=0
      color {colorUseRule=UseColor_Specified, specifiedColor=<rgba 0 0 0 0>}
      cornerRadii=<0 0 0 0>
     }
     tintColor {colorUseRule=UseColor_Specified, specifiedColor=<rgba 1 1 1 1>}
     uVRegion {bIsValid=false, max=<0 0>, min=<0 0>}
    }}
   slot {class=ScaleBoxSlot}
  }
  Image_Shadow {
   widgetClassPath=&Image
   bIsVariable=true
   parent=VerticalBox_1
   widget=Image_Shadow
   properties {
    brush {
     bIsDynamicallyLoaded=false
     drawAs=Image
     imageType=NoImage
     mirroring=NoMirror
     resourceName=None
     resourceObject=/Game/Athena/HUD/LTM/Wax/Materials/M_UI_TextGlow
     tiling=NoTile
     imageSize=<1600 400>
     margin=<ltrb 0 0 0 0>
     outlineSettings {
      bUseBrushTransparency=false
      roundingType=HalfHeightRadius
      width=0
      color {colorUseRule=UseColor_Specified, specifiedColor=<rgba 0 0 0 0>}
      cornerRadii=<0 0 0 0>
     }
     tintColor {colorUseRule=UseColor_Specified, specifiedColor=<rgba 1 1 1 1>}
     uVRegion {bIsValid=false, max=<0 0>, min=<0 0>}
    }
    colorAndOpacity=<rgba 0.0238 0 0.0312 0.8>
   }
   slot {class=VerticalBoxSlot, properties {horizontalAlignment=HAlign_Center, verticalAlignment=VAlign_Center, padding=<ltrb -800 -250 -800 -200>}}
  }
  TutorialHighlight {widgetClassPath=&TutorialHighlight}
 ]
}
